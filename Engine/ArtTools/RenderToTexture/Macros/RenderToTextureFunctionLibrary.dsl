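; Render To Texture macro function library with two helpers: an HLSL int-array string builder and a canvas material placement calculator.
; ArraytoHLSLIntArray switches on component count to serialize the matching input array as an int to int4 declaration, truncating floats toward zero.
; As wired, the last scalar writes its index instead of its value, and int3 and int4 omit one component comma.
; SetCanvasMaterialScaleandPosition converts normalized position and scale into offset and scaled size.
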(fn ArraytoHLSLIntArray
  (params (Type :type EIntTypes%20Enum :default "NewEnumerator4") (Variable%20Name :type String%20%28by%20ref%29) (int :type Array%20of%20Integers) (int2 :type Array%20of%20Vector%202D%20Structures) (int3 :type Array%20of%20Vectors) (int4 :type Array%20of%20Linear%20Color%20Structures))
  (Utilities|FlowControl|Sequence
    (:then_0
      (switch Utilities|FlowControl|Switch|SwitchonEIntTypes Type
        (:NewEnumerator4
          ;[1 1 1] var int[n] = {
          (LocalVariables|Setcurstring (Utilities|String|Append (Utilities|String|Append (Utilities|String|Append "int " Variable%20Name) (Utilities|String|BuildString(Integer) "" "[" (Utilities|Array|Length int) "]")) " = {"))
          ;[1 1 1] n1,n2,n3}
          (for (_array_element _array_index) int
            (LocalVariables|Setcurstring (select (== _array_index (Utilities|Array|LastIndex int)) (Utilities|String|BuildString(Integer) (LocalVariables|Getcurstring) "" _array_index "};") (Utilities|String|BuildString(Integer) (LocalVariables|Getcurstring) "" _array_element ", ")))))
        (:NewEnumerator5
          ;[1 1 1] var int2[n] = {
          (LocalVariables|Setcurstring (Utilities|String|Append (Utilities|String|Append (Utilities|String|Append "int2 " Variable%20Name) (Utilities|String|BuildString(Integer) "" "[" (Utilities|Array|Length int2) "]")) " = {"))
          ;[1 1 1] int2(a,b),int2(c,d),int2(e,f)};
          (for (_array_element_2 _array_index_2) int2
            (LocalVariables|Setcurstring (select (== _array_index_2 (Utilities|Array|LastIndex int2)) (Utilities|String|Append (Utilities|String|Append (Utilities|String|Append (Utilities|String|Append (Utilities|String|Append (LocalVariables|Getcurstring) "int2(") (Utilities|String|ToString(Text) (Utilities|Text|ToText(Float) (Math|Vector2D|BreakVector2D _array_element_2) "ToZero" false true 1 324 0 0))) ",") (Utilities|String|ToString(Text) (Utilities|Text|ToText(Float) (.Y (Math|Vector2D|BreakVector2D _array_element_2)) "ToZero" false true 1 324 0 0))) ")};") (Utilities|String|Append (Utilities|String|Append (Utilities|String|Append (Utilities|String|Append (Utilities|String|Append (LocalVariables|Getcurstring) "int2(") (Utilities|String|ToString(Text) (Utilities|Text|ToText(Float) (Math|Vector2D|BreakVector2D _array_element_2) "ToZero" false true 1 324 0 0))) ",") (Utilities|String|ToString(Text) (Utilities|Text|ToText(Float) (.Y (Math|Vector2D|BreakVector2D _array_element_2)) "ToZero" false true 1 324 0 0))) "), ")))))
        (:NewEnumerator6
          ;[1 1 1] var int3[n] = {
          (LocalVariables|Setcurstring (Utilities|String|Append (Utilities|String|Append (Utilities|String|Append "int3 " Variable%20Name) (Utilities|String|BuildString(Integer) "" "[" (Utilities|Array|Length int3) "]")) " = {"))
          ;[1 1 1] int3(a,b,c),int3(c,d,e),int3(e,f,g)};
          (for (_array_element_3 _array_index_3) int3
            (LocalVariables|Setcurstring (select (== _array_index_3 (Utilities|Array|LastIndex int3)) (Utilities|String|Append (Utilities|String|Append (Utilities|String|Append (Utilities|String|Append (Utilities|String|Append (Utilities|String|Append (LocalVariables|Getcurstring) "int3(") (Utilities|String|ToString(Text) (Utilities|Text|ToText(Float) (.x _array_element_3) "ToZero" false true 1 324 0 0))) ",") (Utilities|String|ToString(Text) (Utilities|Text|ToText(Float) (.y _array_element_3) "ToZero" false true 1 324 0 0))) (Utilities|String|ToString(Text) (Utilities|Text|ToText(Float) (.z _array_element_3) "ToZero" false true 1 324 0 0))) ")};") (Utilities|String|Append (Utilities|String|Append (Utilities|String|Append (Utilities|String|Append (Utilities|String|Append (Utilities|String|Append (LocalVariables|Getcurstring) "int3(") (Utilities|String|ToString(Text) (Utilities|Text|ToText(Float) (.x _array_element_3) "ToZero" false true 1 324 0 0))) ",") (Utilities|String|ToString(Text) (Utilities|Text|ToText(Float) (.y _array_element_3) "ToZero" false true 1 324 0 0))) (Utilities|String|ToString(Text) (Utilities|Text|ToText(Float) (.z _array_element_3) "ToZero" false true 1 324 0 0))) "), ")))))
        (:NewEnumerator7
          ;[1 1 1] var int4[n] = {
          (LocalVariables|Setcurstring (Utilities|String|Append (Utilities|String|Append (Utilities|String|Append "int4 " Variable%20Name) (Utilities|String|BuildString(Integer) "" "[" (Utilities|Array|Length int4) "]")) " = {"))
          ;[1 1 1] int4(a,b,c),....)};
          (for (_array_element_4 _array_index_4) int4
            (LocalVariables|Setcurstring (select (== _array_index_4 (Utilities|Array|LastIndex int4)) (Utilities|String|Append (Utilities|String|Append (Utilities|String|Append (Utilities|String|Append (Utilities|String|Append (Utilities|String|Append (Utilities|String|Append (Utilities|String|Append (LocalVariables|Getcurstring) "int4(") (Utilities|String|ToString(Text) (Utilities|Text|ToText(Float) (Math|Color|BreakColor _array_element_4) "ToZero" false true 1 324 0 0))) ",") (Utilities|String|ToString(Text) (Utilities|Text|ToText(Float) (.G (Math|Color|BreakColor _array_element_4)) "ToZero" false true 1 324 0 0))) (Utilities|String|ToString(Text) (Utilities|Text|ToText(Float) (.B (Math|Color|BreakColor _array_element_4)) "ToZero" false true 1 324 0 0))) ",") (Utilities|String|ToString(Text) (Utilities|Text|ToText(Float) (.A (Math|Color|BreakColor _array_element_4)) "ToZero" false true 1 324 0 0))) ")};") (Utilities|String|Append (Utilities|String|Append (Utilities|String|Append (Utilities|String|Append (Utilities|String|Append (Utilities|String|Append (Utilities|String|Append (Utilities|String|Append (LocalVariables|Getcurstring) "int4(") (Utilities|String|ToString(Text) (Utilities|Text|ToText(Float) (Math|Color|BreakColor _array_element_4) "ToZero" false true 1 324 0 0))) ",") (Utilities|String|ToString(Text) (Utilities|Text|ToText(Float) (.G (Math|Color|BreakColor _array_element_4)) "ToZero" false true 1 324 0 0))) (Utilities|String|ToString(Text) (Utilities|Text|ToText(Float) (.B (Math|Color|BreakColor _array_element_4)) "ToZero" false true 1 324 0 0))) ",") (Utilities|String|ToString(Text) (Utilities|Text|ToText(Float) (.A (Math|Color|BreakColor _array_element_4)) "ToZero" false true 1 324 0 0))) "), ")))))))
    (:then_1
      (return (LocalVariables|Getcurstring)))))

(fn SetCanvasMaterialScaleandPosition
  (params (Size :type Vector%202D%20Structure) (Position :type Vector%202D%20Structure) (Scale :type Float%20%28double-precision%29))
  (return (+ (Math|Conversions|ToVector2D(Vector) (* (Math|Conversions|ToVector(Vector2D) Size) (Math|Conversions|ToVector(Vector2D) (- Position 0.5)))) (* (* Size 0.5) (- 1 Scale))) (* Size Scale)))
